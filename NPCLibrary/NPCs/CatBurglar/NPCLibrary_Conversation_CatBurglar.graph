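# Conversation database for the CatBurglar NPC in Fortnite, defining its greeting and top-level menu when a player talks to it.
# The entry tag leads to one speech node with Midas-themed lines, chosen by contextual requirements for first-ever meeting, specific player outfits or per-match conversation count, with curve-row priorities on the first two kinds.
# That node branches through link nodes to shared quest, basic-service and Season 29 activation conversations, plus CatBurglar's own dialogue entry.
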ConversationDatabase /NPCLibrary/NPCs/CatBurglar/NPCLibrary_Conversation_CatBurglar
entry_points [{tag=Conversation.Entry.CatBurglar, destinations [#1]}]
nodes [
 ConversationEntryPointNode_7 {
  id=#1
  class=ConversationEntryPointNode
  outputs [#2]
  properties {entryTag {tagName=Conversation.Entry.CatBurglar}}
 }
 Speech_C_4 {
  id=#2
  class=Speech_C
  outputs [#3, #4, #5, #6]
  properties {
   nodeName=""
   generalConfig {
    defaultMessage=I told you I'd have work for you.
    contextualMessages [
     {
      message="It's been a while, hasn't it?"
      contextRequirements [{requirement {class=FortControllerRequirement_IsFirstEverConversationWithNPC}}]
      priority {curve {rowName=Tandem.Speech.Priority.FirstEverInteraction}}
     }
     {
      message=I told you he was here. Wait on my signal - we're getting him off this Island.
      contextRequirements [{requirement {
         class=FortControllerRequirement_HasCID
         properties {allowedCIds [
           &CID_692_Athena_Commando_M_HenchmanTough
           /BRCosmetics/Athena/Items/Cosmetics/Characters/CID_690_Athena_Commando_F_Photographer
           /BRCosmetics/Athena/Items/Cosmetics/Characters/CID_691_Athena_Commando_F_TNTina
           &CID_693_Athena_Commando_M_BuffCat
          ]}
        }}]
      priority {curve {rowName=Tandem.Speech.Priority.OutfitReaction}}
     }
     {
      message="Give your father a chance. Midas adores you, in his way."
      contextRequirements [{requirement {
         class=FortControllerRequirement_HasCID
         properties {allowedCIds [/BRCosmetics/Athena/Items/Cosmetics/Characters/CID_770_Athena_Commando_F_MechanicalEngineer]}
        }}]
      priority {curve {rowName=Tandem.Speech.Priority.OutfitReaction}}
     }
     {
      message=Thank you. I couldn't have found him without you.
      contextRequirements [{requirement {
         class=FortControllerRequirement_HasCID
         properties {allowedCIds [/BRCosmetics/Athena/Items/Cosmetics/Characters/CID_A_135_Athena_Commando_F_StreetFashionSummer]}
        }}]
      priority {curve {rowName=Tandem.Speech.Priority.OutfitReaction}}
     }
     {
      message=I can't believe you named a boat after me.
      contextRequirements [{requirement {class=FortControllerRequirement_HasCID, properties {allowedCIds [/BRCosmetics/Athena/Items/Cosmetics/Characters/Character_PiteousKicks]}}}]
      priority {curve {rowName=Tandem.Speech.Priority.OutfitReaction}}
     }
     {
      message=I told you I'd have work for you.
      contextRequirements [{requirement {class=FortControllerRequirement_CheckNPCMatchConversationCount, properties {nPCUniqueIdTag {tagName=AISpawnerData.Type.Tandem.CatBurglar}}}}]
      priority {curve {rowName=None}}
     }
     {
      message=At long last... he's back.
      contextRequirements [{requirement {class=FortControllerRequirement_CheckNPCMatchConversationCount, properties {nPCUniqueIdTag {tagName=AISpawnerData.Type.Tandem.CatBurglar}}}}]
      priority {curve {rowName=None}}
     }
     {
      message="It has been a long time, golden king..."
      contextRequirements [{requirement {class=FortControllerRequirement_CheckNPCMatchConversationCount, properties {nPCUniqueIdTag {tagName=AISpawnerData.Type.Tandem.CatBurglar}}}}]
      priority {curve {rowName=None}}
     }
     {
      message=There is only one person who understands the Golden Touch like I do.
      contextRequirements [{requirement {class=FortControllerRequirement_CheckNPCMatchConversationCount, properties {nPCUniqueIdTag {tagName=AISpawnerData.Type.Tandem.CatBurglar}}}}]
      priority {curve {rowName=None}}
     }
     {
      message=I've kept my promise to the crew. What happens now is up to Midas.
      contextRequirements [{requirement {class=FortControllerRequirement_CheckNPCMatchConversationCount, properties {nPCUniqueIdTag {tagName=AISpawnerData.Type.Tandem.CatBurglar}}}}]
      priority {curve {rowName=None}}
     }
    ]
   }
  }
 }
 ConversationLinkNode_3 {
  id=#6
  class=ConversationLinkNode
  properties {remoteEntryTag {tagName=Conversation.Entry.CatBurglar.Dialogue}}
 }
 ConversationLinkNode_22 {
  id=#5
  class=ConversationLinkNode
  properties {remoteEntryTag {tagName=Conversation.Entry.Season29.SeasonalActivation.01b}}
 }
 ConversationLinkNode_0 {id=#4, class=ConversationLinkNode, properties {remoteEntryTag {tagName=Conversation.Entry.BasicServices}}}
 ConversationLinkNode_1 {id=#3, class=ConversationLinkNode, properties {remoteEntryTag {tagName=Conversation.Entry.NPCLibrary.Quests}}}
]
